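# Translucent unlit sprite material for the guided tutorial X marker, usable on Niagara and particle sprites.
# The marker texture's green channel, tinted by Color and Emissive Intensity, feeds an emissive adjustment function.
# An opacity chain multiplies green, Opacity, power-adjusted red and hard-edged scanlines from a panned UV sine and SmoothStep.
# Scanlines are off by default; a clamped time-sine chain has no visible consumer.
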Material /Game/Effects/Fort_Effects/Gameplay/Tutorial_Target/M_GuidedTutorial_X_marker
graph {
 properties {
  bUsedWithNiagaraSprites=true
  bUsedWithParticleSprites=true
  blendMode=BLEND_Translucent
  pixelDepthOffsetMode=PDOM_Legacy
  shadingModel=MSM_Unlit
  usedShadingModels=Unlit
 }
 nodes [
  n1 = Time()
  n2 = Constant() {r=1}
  Clamp(None=Sine(None=Multiply(A=@n1, B=@n2)), Min=Constant(), Max=Constant() {r=5})
  n3 = VectorParameter() {parameterName=Color, defaultValue=<rgba 1 1 1 1>}
  n4 = TextureSample() {samplerType=SAMPLERTYPE_LinearColor, texture=&/T_GuidedTutorial_X_marker}
  MaterialFunctionCall(Emissive=Multiply(A=Multiply(A=@n3.RGB, B=@n4.G), B=ScalarParameter() {defaultValue=1, parameterName=Emissive Intensity})) {materialFunction=&MF_FX_EmissiveAdjustments}
  n5 = Multiply(A=TextureCoordinate(), B=AppendVector(A=Constant() {r=1}, B=ScalarParameter() {defaultValue=20, parameterName=Scanlines Tiling}))
  n6 = AppendVector(A=Constant(), B=ScalarParameter() {parameterName=Scanlines Speed})
  n7 = Sine(None=ComponentMask(None=Panner(Coordinate=@n5, Speed=@n6)) {g=true})
  n8 = Constant() {r=0.5}
  n9 = Constant() {r=0.5}
  n10 = LinearInterpolate(
   A=Constant() {r=1}
   B=OneMinus(None=MaterialFunctionCall(Alpha=@n7, Min=@n8, Max=@n9) {materialFunction=&SmoothStep}.Result)
   Alpha=ScalarParameter() {parameterName=Scanlines Opacity}
  )
  n11 = ScalarParameter() {defaultValue=1, parameterName=Power}
  Multiply(
   A=Multiply(A=Multiply(A=@n4.G, B=@n10), B=ScalarParameter() {defaultValue=1, parameterName=Opacity})
   B=Clamp(None=Power(Base=@n4.R, Exp=@n11) {constExponent=1})
  )
 ]
}
