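# Translucent unlit, two-sided particle material that draws tinted chevrons which blink in stepped bands, apparently for tutorial target markers.
# The Jump Timing dynamic parameter, quantized with Frac and Ceil, offsets a banded V gradient that modulates emissive and opacity when Animate Arrows is on.
# Opacity combines chevron texture, particle alpha and switchable vertical, radial and scanline masks, with an outline blend.
# An optional sine-driven world position offset adds Z-axis bobbing.
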Material /Game/Effects/Fort_Effects/Gameplay/Tutorial_Target/M_Tutorial_BlinkingArrows
graph {
 properties {
  bUseLightmapDirectionality=false
  bUseTranslucencyVertexFog=false
  bUsedWithMeshParticles=true
  bUsedWithNiagaraSprites=true
  bUsedWithParticleSprites=true
  blendMode=BLEND_Translucent
  pixelDepthOffsetMode=PDOM_Legacy
  shadingModel=MSM_Unlit
  twoSided=true
  usedShadingModels=Unlit
 }
 nodes [
  n1 = TextureCoordinate()
  n2 = ComponentMask(None=OneMinus(None=@n1)) {g=true}
  n3 = ScalarParameter() {defaultValue=1, parameterName=Time Scale}
  n4 = DynamicParameter() {paramNames [Opacity Over Lifetime, Jump Timing, Lifetime, Param4]}
  n5 = Multiply(A=Frac(None=Multiply(A=@n3, B=@n4.Jump Timing)), B=ScalarParameter() {defaultValue=5, parameterName=Time steps Multiplier}) {constB=5}
  n6 = Multiply(A=Divide(A=Ceil(None=@n5), B=ScalarParameter() {defaultValue=0.5, parameterName=Time steps Divider}) {constB=0.5}) {constB=0.1}
  n7 = Multiply(A=Frac(None=Add(A=@n2, B=@n6)), B=ScalarParameter() {defaultValue=5, parameterName=Jump Steps}) {constB=9}
  n8 = MaterialFunctionCall(
   Input=Divide(A=Ceil(None=@n7)) {constB=9}
   Input Low=ScalarParameter() {parameterName=Input Low}
   Input High=ScalarParameter() {defaultValue=1, parameterName=Input High}
   Target Low=ScalarParameter() {parameterName=Target Low}
   Target High=ScalarParameter() {defaultValue=1, parameterName=Target High}
  ) {materialFunction=&RemapValueRange}
  n9 = Add(
   A=Multiply(A=ComponentMask(None=@n1) {r=true}, B=ScalarParameter() {defaultValue=1, parameterName=X tiling})
   B=ScalarParameter() {parameterName=X Offset}
  )
  n10 = ComponentMask(None=@n1) {g=true}
  n11 = Add(A=Multiply(A=@n10, B=ScalarParameter() {defaultValue=1, parameterName=Y tiling}), B=ScalarParameter() {parameterName=Y Offset})
  n12 = TextureSampleParameter2D(UVs=AppendVector(A=@n9, B=@n11)) {
   parameterName=Texture Input
   samplerType=SAMPLERTYPE_LinearColor
   texture=&/T_Tutorial_Marker_Chevron
  }
  n13 = VectorParameter() {parameterName=Tint Color, defaultValue=<rgba 3 1.0439 0 1>}
  n14 = Multiply(A=@n12.RGB, B=@n13.RGB)
  n15 = Multiply(A=@n8.Result, B=@n14)
  n16 = ParticleColor()
  MaterialFunctionCall(Emissive=Multiply(A=StaticSwitchParameter(True=@n15, False=@n14) {parameterName=Animate Arrows}, B=@n16.RGB)) {materialFunction=&MF_FX_EmissiveAdjustments}
  n17 = Multiply(A=@n12.R, B=@n16.A)
  Multiply(A=@n13.RGB, B=@n17)
  n18 = Constant()
  n19 = Distance(A=@n10, B=@n18)
  n20 = OneMinus(None=@n19)
  n21 = StaticSwitchParameter(
   True=Add(A=StaticSwitchParameter(True=@n20, False=@n19) {parameterName=Flip Y Gradient}, B=ScalarParameter() {parameterName=Vertical Mask Transition})
   False=Constant() {r=1}
  ) {parameterName=Use top-bottom gradient}
  n22 = Multiply(A=Multiply(A=@n13.A, B=@n17), B=@n21) {constA=1}
  Reroute(None=@n22)
  n23 = StaticSwitchParameter(True=Power(Base=@n8.Result, Exp=ScalarParameter() {defaultValue=2, parameterName=Arrow Opacity strength}), False=Constant() {r=1}) {parameterName=Animate Arrows}
  n24 = Subtract(A=@n1) {constB=0.5}
  n25 = Power(Base=OneMinus(None=Distance(A=@n24, B=@n18)), Exp=ScalarParameter() {defaultValue=1, parameterName=Radial Mask Sharpness})
  n26 = Constant() {r=1}
  n27 = Multiply(A=Multiply(A=@n22, B=StaticSwitchParameter(True=@n25, False=@n26) {parameterName=Use Radial gradient}), B=@n21)
  n28 = StaticSwitchParameter(True=AppendVector(A=@n11, B=OneMinus(None=@n9)), False=ScreenPosition().ViewportUV) {defaultValue=true, parameterName=Scaline use UV}
  n29 = MaterialFunctionCall(Float2=@n28) {materialFunction=&BreakOutFloat2Components}
  n30 = MaterialFunctionCall!(
   Number of Scanlines
   Scanline Speed
   "Scanlines Thickness (Bias)"
   Scanlines Dark Value
   Scanlines Light Value
   Block Opacity
   UVs=StaticSwitchParameter(True=AppendVector(A=@n29.G, B=@n29.G), False=@n28) {parameterName=Invert UVs X-Y}
   Time
   Block Repeat=Constant() {r=-0.25}
  ) {materialFunction=/Game/Effects/Fort_Effects/Effects/Atlas/Materials/MF_Atlas_Scanlines_01}
  n31 = Add(A=@n30.Result, B=@n30.Result)
  n32 = Constant() {r=1}
  n33 = Clamp(None=Multiply(A=Multiply(A=@n23, B=@n27), B=StaticSwitchParameter(True=@n31, False=@n32) {parameterName=Use scalines}))
  n34 = Multiply(A=@n33) {constB=0.5}
  n35 = ComponentMask(None=@n12.R) {r=true}
  n36 = Constant() {r=-0.1}
  n37 = Multiply(A=@n12.R, B=OneMinus(None=Distance(A=@n35, B=@n36)))
  n38 = Multiply(A=@n27, B=Clamp(None=Multiply(A=@n37) {constB=5}))
  Saturate(None=LinearInterpolate(A=Add(A=@n34, B=@n38), B=@n33, Alpha=ScalarParameter() {parameterName=Use Outline?}))
  n39 = Constant2Vector()
  n40 = Sine(None=Time())
  n41 = ScalarParameter() {defaultValue=10, parameterName=WPO sine multiplier}
  n42 = ComponentMask(None=Add(A=Multiply(A=@n40, B=@n41)) {constB=3}) {b=true}
  n43 = Constant3Vector()
  StaticSwitchParameter(True=LinearInterpolate(A=AppendVector(A=@n39, B=@n42), B=@n43, Alpha=@n40), False=@n43) {parameterName=Use WPO}
  n44 = MaterialFunctionCall(Float2=ScreenPosition().ViewportUV) {materialFunction=&BreakOutFloat2Components}
  n45 = AppendVector(A=Multiply(A=@n44.R) {constB=2}, B=@n44.G)
  n46 = Sine(None=Multiply(A=Time()) {constB=10})
  TextureSample(UVs=Panner(Coordinate=Multiply(A=@n45) {constB=3.3}, Time=Ceil(None=@n46)) {speedX=0.5, speedY=0.5}) {samplerType=SAMPLERTYPE_LinearColor, texture=/Game/UI/Foundation/Textures/Pattern/T_CubePattern}
 ]
}
