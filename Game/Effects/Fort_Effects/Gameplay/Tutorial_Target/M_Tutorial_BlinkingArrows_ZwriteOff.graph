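# Unlit, two-sided translucent particle material with depth test disabled, drawing tinted chevron arrows for the tutorial target effect.
# A tiled, offsettable texture sample combines with tint and particle color, using alpha or RGB as color before an emissive-adjustment function.
# An alpha chain applies an optional vertical gradient, then blends toward a radially masked version via the Particle Lifetime dynamic parameter.
# The graph has no time node, so any blinking comes from outside it.
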Material /Game/Effects/Fort_Effects/Gameplay/Tutorial_Target/M_Tutorial_BlinkingArrows_ZwriteOff
graph {
 properties {
  bDisableDepthTest=true
  bUseLightmapDirectionality=false
  bUseTranslucencyVertexFog=false
  bUsedWithMeshParticles=true
  bUsedWithNiagaraSprites=true
  bUsedWithParticleSprites=true
  blendMode=BLEND_Translucent
  pixelDepthOffsetMode=PDOM_Legacy
  shadingModel=MSM_Unlit
  twoSided=true
  usedShadingModels=Unlit
 }
 nodes [
  n1 = VectorParameter() {parameterName=Tint Color, defaultValue=<rgba 4 1.3919 0 1>}
  n2 = TextureCoordinate()
  n3 = Multiply(A=ComponentMask(None=@n2) {r=true}, B=ScalarParameter() {defaultValue=1, parameterName=X tiling})
  n4 = ScalarParameter() {parameterName=X Offset}
  n5 = ComponentMask(None=@n2) {g=true}
  n6 = Multiply(A=@n5, B=ScalarParameter() {defaultValue=3, parameterName=Y tiling})
  n7 = ScalarParameter() {parameterName=Y Offset}
  n8 = TextureSampleParameter2D(UVs=AppendVector(A=Add(A=@n3, B=@n4), B=Add(A=@n6, B=@n7))) {
   parameterName=Texture Input
   samplerType=SAMPLERTYPE_LinearColor
   texture=&/T_Tutorial_Marker_Chevron
  }
  n9 = Multiply(A=@n8.A, B=ParticleColor().A)
  n10 = Constant()
  n11 = Distance(A=@n5, B=@n10)
  n12 = OneMinus(None=@n11)
  n13 = ScalarParameter() {parameterName=Flip Y Gradient, sliderMax=1}
  n14 = LinearInterpolate(
   A=Constant() {r=1}
   B=Add(A=LinearInterpolate(A=@n12, B=@n11, Alpha=@n13), B=ScalarParameter() {parameterName=Vertical Mask Transition})
   Alpha=ScalarParameter() {parameterName=Use Top-Bottom Gradient}
  )
  n15 = Multiply(A=Multiply(A=@n1.A, B=@n9), B=@n14) {constA=1}
  n16 = Distance(A=Subtract(A=@n2) {constB=0.5}, B=@n10)
  n17 = LinearInterpolate(
   A=Power(Base=OneMinus(None=@n16), Exp=ScalarParameter() {defaultValue=1, parameterName=Radial Mask Sharpness})
   B=Constant() {r=1}
   Alpha=ScalarParameter() {parameterName=Use Radial Gradient}
  )
  LinearInterpolate(
   A=@n15
   B=Multiply(A=Multiply(A=@n15, B=@n17), B=@n14)
   Alpha=DynamicParameter() {paramNames [Particle Lifetime, Param2, Param3, Param4]}.Particle Lifetime
  )
  n18 = Multiply(A=@n1.RGB, B=@n9)
  n19 = Multiply(A=@n8.RGB, B=@n1.RGB)
  n20 = ScalarParameter() {parameterName=Use Alpha as Color}
  MaterialFunctionCall(Emissive=Multiply(A=ParticleColor().RGB, B=LinearInterpolate(A=@n18, B=@n19, Alpha=@n20))) {materialFunction=&MF_FX_EmissiveAdjustments}
 ]
}
