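# Unlit translucent particle material for a tutorial target marker, usable on mesh and sprite particles.
# A vertical two-color gradient blends toward a bright tint band at mid-V, and a scrolling sine wave is added before particle color and emissive adjustments.
# Opacity combines the V coordinate, tint alpha and an axis-aligned fresnel on vertex normals, softened by depth fade and scaled by particle alpha.
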Material /Game/Effects/Fort_Effects/Gameplay/Tutorial_Target/M_Tutorial_Marker
graph {
 properties {
  bUseTranslucencyVertexFog=false
  bUsedWithMeshParticles=true
  bUsedWithParticleSprites=true
  blendMode=BLEND_Translucent
  pixelDepthOffsetMode=PDOM_Legacy
  shadingModel=MSM_Unlit
  translucencyPass=MTP_BeforeDOF
  usedShadingModels=Unlit
 }
 nodes [
  n1 = Multiply(A=Time(), B=Constant() {r=-0.1})
  n2 = ComponentMask(None=TextureCoordinate()) {g=true}
  n3 = Sine(None=Multiply(A=Add(A=@n1, B=@n2)) {constB=4})
  n4 = VectorParameter() {parameterName=Top Color, defaultValue=<rgba 0.5377 0.8265 1 1>}
  n5 = VectorParameter() {parameterName=Bottom Color, defaultValue=<rgba 0.5 0.2592 0.4151 1>}
  n6 = ComponentMask(None=TextureCoordinate()) {g=true}
  n7 = VectorParameter() {parameterName=Tint Color, defaultValue=<rgba 19.8958 20 19.7917 0.5>}
  n8 = OneMinus(None=Distance(A=Subtract(A=@n6) {constB=0.5}, B=Constant()))
  n9 = ScalarParameter() {defaultValue=3, parameterName=Color Map Exponent}
  n10 = LinearInterpolate(
   A=ComponentMask(None=LinearInterpolate(A=@n4.RGB, B=@n5.RGB, Alpha=@n6)) {b=true, g=true, r=true}
   B=ComponentMask(None=@n7.RGB) {b=true, g=true, r=true}
   Alpha=Saturate(None=Power(Base=@n8, Exp=@n9))
  )
  n11 = ParticleColor()
  MaterialFunctionCall(Emissive=Multiply(A=Add(A=@n3, B=@n10), B=@n11.RGB)) {materialFunction=&MF_FX_EmissiveAdjustments}
  n12 = Multiply(
   A=@n7.A
   B=MaterialFunctionCall(Exponent=ScalarParameter() {defaultValue=2, parameterName=Fresnel Exponent}, Normal=VertexNormalWS()) {materialFunction=&AxisAlignedFresnel}.Fresnel
  )
  Multiply(A=DepthFade(Opacity=Multiply(A=@n2, B=@n12), FadeDistance=ScalarParameter() {defaultValue=1000, parameterName=Fade Distance}), B=@n11.A)
 ]
}
