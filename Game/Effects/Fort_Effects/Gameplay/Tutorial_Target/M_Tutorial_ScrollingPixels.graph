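# Translucent mesh-particle material for a tutorial target effect, showing a pixel texture that scrolls vertically with time.
# Two offset samples of the texture sum, then particle color and a tint or optional color gradient shape emissive, while a positioned gradient fades opacity.
# A switchable bottom gradient, whose falloff pulses on a sine, adds to both color and opacity.
# Optional scanlines and depth fade further modify opacity.
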Material /Game/Effects/Fort_Effects/Gameplay/Tutorial_Target/M_Tutorial_ScrollingPixels
graph {
 properties {
  bUseLightmapDirectionality=false
  bUseTranslucencyVertexFog=false
  bUsedWithMeshParticles=true
  blendMode=BLEND_Translucent
  pixelDepthOffsetMode=PDOM_Legacy
  usedShadingModels=Default Lit
 }
 nodes [
  n1 = Distance(A=ComponentMask(None=TextureCoordinate()) {g=true}, B=Constant())
  n2 = Multiply(A=ScalarParameter() {defaultValue=0.1, parameterName=Sine Speed}, B=Time())
  n3 = LinearInterpolate(
   A=@n1
   B=Power(Base=@n1, Exp=ScalarParameter() {defaultValue=10, parameterName=Top-Down Gradient Power})
   Alpha=ConstantBiasScale(None=Sine(None=@n2)) {bias=6.5, scale=0.2}
  )
  n4 = VectorParameter() {parameterName=Top Gradient Color, defaultValue=<rgba 0.9115 0 0.3734 1>}
  n5 = VectorParameter() {parameterName=Bottom Gradient Color, defaultValue=<rgba 0 0.2525 0.349 1>}
  n6 = VectorParameter() {parameterName=Tint Color, defaultValue=<rgba 1 1 1 0>}
  n7 = TextureCoordinate()
  n8 = ComponentMask(None=@n7) {g=true}
  n9 = ScalarParameter() {defaultValue=-0.5, parameterName=Opacity Gradient Position}
  n10 = OneMinus(None=Distance(A=Add(A=@n8, B=@n9) {constB=-0.5}, B=Constant()))
  n11 = StaticSwitchParameter(
   True=LinearInterpolate(A=LinearInterpolate(A=@n4.RGB, B=@n5.RGB, Alpha=@n1), B=@n6.RGB, Alpha=Power(Base=@n10) {constExponent=6})
   False=@n6.RGB
  ) {parameterName=Use color gradient}
  n12 = Multiply(A=@n3, B=@n11)
  n13 = Multiply(A=Time(), B=ScalarParameter() {defaultValue=1, parameterName=Time Scale})
  n14 = AppendVector(
   A=Multiply(A=ComponentMask(None=@n7) {r=true}, B=ScalarParameter() {defaultValue=1, parameterName="X Tiling "})
   B=Multiply(A=Add(A=@n8, B=@n13), B=ScalarParameter() {defaultValue=1, parameterName=Y Tiling})
  )
  n15 = TextureObjectParameter() {
   parameterName=Main Texture
   samplerType=SAMPLERTYPE_LinearGrayscale
   texture=/Game/UI/Frontend/ItemShop/Com/vfx/T_CommunityVoting_Activities_Pixels.CommunityVoting_Activities_Pixels
  }
  n16 = TextureSample(UVs=Add(A=@n14) {constB=0.5}, Tex=@n15) {
   samplerType=SAMPLERTYPE_LinearGrayscale
   texture=&CommunityVoting_Activities_Pixels
  }
  n17 = TextureSample(UVs=@n14, Tex=@n15) {
   samplerType=SAMPLERTYPE_LinearGrayscale
   texture=&CommunityVoting_Activities_Pixels
  }
  n18 = Add(A=@n16.RGB, B=@n17.RGB)
  n19 = ScalarParameter() {defaultValue=1, parameterName=Pixels Opacity, sliderMax=1}
  n20 = ParticleColor()
  n21 = Multiply(A=@n11, B=Multiply(A=Multiply(A=@n18, B=@n19), B=@n20.RGB))
  MaterialFunctionCall(Emissive=StaticSwitchParameter(True=Add(A=@n12, B=@n21), False=@n21) {parameterName=Use bottom gradient}) {materialFunction=&MF_FX_EmissiveAdjustments}
  n22 = Multiply(A=Add(A=@n16.A, B=@n17.A), B=@n19)
  n23 = Multiply(
   A=Multiply(A=@n6.A, B=Multiply(A=@n22, B=@n20.A))
   B=Power(Base=@n10, Exp=ScalarParameter() {defaultValue=1, parameterName=Opacity Gradient Power})
  )
  n24 = StaticSwitchParameter(True=Add(A=Multiply(A=@n3, B=@n6.A), B=@n23), False=@n23) {parameterName=Use bottom gradient}
  n25 = Multiply(
   A=@n24
   B=MaterialFunctionCall!(
    Number of Scanlines
    Scanline Speed
    "Scanlines Thickness (Bias)"
    Scanlines Dark Value
    Scanlines Light Value
    Block Opacity
    UVs=TextureCoordinate()
    Time
    Block Repeat
   ) {materialFunction=/Game/Effects/Fort_Effects/Effects/Atlas/Materials/MF_Atlas_Scanlines_01}.Result
  )
  n26 = StaticSwitchParameter(True=Saturate(None=Add(A=@n24, B=@n25)), False=@n24) {parameterName=Use Scanlines}
  StaticSwitchParameter(True=DepthFade(Opacity=@n26, FadeDistance=ScalarParameter() {defaultValue=400, parameterName=Fade Distance}), False=@n26) {parameterName=Use Depth Fade}
 ]
}
